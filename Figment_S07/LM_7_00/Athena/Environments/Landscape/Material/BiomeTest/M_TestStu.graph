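# Test material for the arid biome, usable on spline meshes, that blends desert sand with gravel.
# Sand color and normal use top-down world-aligned projection with a shared size parameter, while gravel color uses tiled mesh UVs.
# HeightLerp blends the tinted colors using scaled vertex alpha as transition phase and gravel alpha as height.
# The normal comes from sand alone, scaled by a strength vector, with no gravel normal.
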Material /Figment_S07/LM_7_00/Athena/Environments/Landscape/Material/BiomeTest/M_TestStu
graph {
 properties {bUsedWithSplineMeshes=true, pixelDepthOffsetMode=PDOM_Legacy, usedShadingModels=Default Lit}
 nodes [
  Constant() {r=1}
  n1 = TextureObject() {texture=/Figment_S07/LM_7_00/Environments/Sets/Arid/Textures/T_desert_sand_base_C}
  n2 = ScalarParameter() {defaultValue=750, parameterName=size}
  n3 = TextureSampleParameter2D(UVs=TextureCoordinate() {uTiling=2, vTiling=2}) {
   group=Gravel
   parameterName=Gravel
   samplerSource=SSM_Wrap_WorldGroupSettings
   texture=/Figment_S07/LM_7_00/Packages/DS_Fortnight/WhiteRockCustom_D
  }
  n4 = VertexColor()
  n5 = ScalarParameter() {defaultValue=0.8476, parameterName=TransMP}
  LinearInterpolate(
   A=Multiply(
    A=MaterialFunctionCall(TextureObject=@n1, TextureSize=@n2) {materialFunction=&WorldAlignedTexture}.Z Texture
    B=VectorParameter() {parameterName=SandTint, defaultValue=<rgba 1 0.833 0.843 1>}.RGB
   )
   B=Multiply(A=@n3.RGB, B=VectorParameter() {parameterName=GravelAridTint, defaultValue=<rgba 2.2 1.3237 0.9783 0>}.RGB)
   Alpha=MaterialFunctionCall(Transition Phase=Multiply(A=@n4.A, B=@n5), Height Texture=@n3.A, Contrast=Constant() {r=5}) {materialFunction=&HeightLerp}.Alpha
  )
  Multiply(
   A=MaterialFunctionCall(
    TextureObject=TextureObject() {samplerType=SAMPLERTYPE_Normal, texture=/Figment_S07/LM_7_00/Environments/Sets/Arid/Textures/T_desert_sand_base_N}
    TextureSize=@n2
   ) {materialFunction=&WorldAlignedNormal}.Z Texture
   B=VectorParameter() {parameterName=SandNormalStrength, defaultValue=<rgba 0.5 0.5 1 0>}.RGB
  )
  ScalarParameter() {parameterName=Spec}
 ]
}
